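# Skinned cloth-shaded character material for the JadeTowelHope Life skin, adding a Boogie Bomb disco sparkle effect to the 2023 anime shader.
# UVs are split into a Scale-sized grid, with each cell hashed for flash timing and a periodically stepped hue.
# An FX channel mask and radial per-cell falloff blend the sparkles over base emissive, and a strength parameter multiplies the result.
# Anime Pre, Calculations and Post functions handle the remaining shading.
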Material /BRCosmetics/Characters/Player/Female/Medium/Bodies/F_MED_JadeTowelHope/Skins/Life/Materials/M_JadeTowelHopeLife_BoogieBomb
graph {
 properties {
  bAutomaticallySetUsageInEditor=false
  bUseMaterialAttributes=true
  bUsedWithMeshDeformer=true
  bUsedWithMorphTargets=true
  bUsedWithSkeletalMesh=true
  bUsedWithStaticLighting=true
  pixelDepthOffsetMode=PDOM_Legacy
  previewMesh=/BRCosmetics/Characters/Player/Female/Medium/Bodies/F_MED_JadeTowelHope/Meshes/F_MED_JadeTowelHope
  shadingModel=MSM_Cloth
  usedShadingModels=Cloth
 }
 nodes [
  Constant()
  n1 = MaterialFunctionCall!(CustomUV) {materialFunction=/Game/Global/Materials/Character_Parent_2023/Anime_Parent_2023/MaterialFunctions/MFC_AnimeShader_2023_Pre}
  n2 = MaterialFunctionCall!(CustomLightVector, OverrideNormal_WS, STT=@n1.STT, Normal_Map=@n1.STT, DFL=@n1.STT, FX=@n1.STT) {materialFunction=/Game/Global/Materials/Character_Parent_2023/Anime_Parent_2023/MaterialFunctions/MFC_AnimeShader_2023_Calculations}
  n3 = ScalarParameter() {defaultValue=128, group=04_BoogieBomb, parameterName=Scale}
  n4 = Multiply(A=TextureCoordinate(), B=@n3)
  n5 = Divide(A=Floor(None=@n4), B=@n3)
  n6 = Constant()
  n7 = MaterialFunctionCall(Float3=MaterialFunctionCall(Vec3=AppendVector(A=@n5, B=@n6)) {materialFunction=&MF_Hash}.Vec3) {materialFunction=&BreakOutFloat3Components}
  n8 = Frac(None=Add(A=Time(), B=@n7.R))
  n9 = Divide(A=Time()) {constB=0.666}
  n10 = Add(A=@n7.G, B=Divide(A=Round(None=@n9)))
  n11 = Multiply(
   A=Multiply(A=@n8, B=Power(Base=@n8) {constExponent=8})
   B=MaterialFunctionCall(Hue Shift Percentage=Frac(None=@n10)) {materialFunction=&HueShift}.Result
  )
  n12 = Distance(A=Subtract(A=Frac(None=@n4)) {constB=0.5}, B=Constant())
  n13 = Multiply(
   A=ChannelMaskParameter(None=@n1.FX) {parameterName=BoogieBomp_ChannelMask}.R
   B=Power(Base=OneMinus(None=@n12)) {constExponent=1, desc=grid mask}
  )
  MaterialFunctionCall!(
   Shading_Gradient=@n2.Shading_Gradient
   Threshold_Gradient=@n2.Shading_Gradient
   Specular_Gradient=@n2.Shading_Gradient
   Sheen_Mask=@n2.Shading_Gradient
   Thickness=@n2.Shading_Gradient
   DFL_Mask=@n2.Shading_Gradient
   FX=@n2.Shading_Gradient
   Metallic_Gradient=@n2.Shading_Gradient
   RimLight_Gradient=@n2.Shading_Gradient
   NormalBuffer=@n2.Shading_Gradient
   --------
   CINE_LightSources=@n2.Shading_Gradient
   CINE_LightColor=@n2.Shading_Gradient
   ---------
   Color_Lit=@n1.Color_Lit
   Color_Shaded=@n1.Color_Lit
   SRM=@n1.Color_Lit
   Emissive=Multiply(A=LinearInterpolate(A=@n1.Emissive, B=@n11, Alpha=@n13), B=ScalarParameter() {defaultValue=40, parameterName=BoogieBomb_EmissiveStrenght})
   DFL_Color=@n1.Color_Lit
   Specular_Color=@n1.Color_Lit
   RimLight_Color=@n1.Color_Lit
  ) {materialFunction=/Game/Global/Materials/Character_Parent_2023/Anime_Parent_2023/MaterialFunctions/MFC_AnimeShader_2023_Post}
 ]
}
